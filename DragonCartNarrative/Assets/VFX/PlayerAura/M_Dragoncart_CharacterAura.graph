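# Translucent unlit material for Niagara sprite and mesh particles, used for the DragonCart character aura.
# A panning turbulence noise distorts the UVs of the base mask, whose green channel drives opacity with particle alpha.
# A vertical UV edge mask, view-angle fade and depth fade soften that opacity.
# Emissive blends a tint with two panning color noises, then applies desaturation, particle color and the shared FX emissive adjustment function.
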Material /DragonCartNarrative/Assets/VFX/PlayerAura/M_Dragoncart_CharacterAura
graph {
 properties {
  bUsedWithNiagaraMeshParticles=true
  bUsedWithNiagaraSprites=true
  blendMode=BLEND_Translucent
  pixelDepthOffsetMode=PDOM_Legacy
  shadingModel=MSM_Unlit
  usedShadingModels=Unlit
 }
 nodes [
  n1 = ComponentMask(None=TextureCoordinate()) {r=true}
  n2 = ComponentMask(None=TextureCoordinate()) {g=true}
  n3 = TextureCoordinate()
  n4 = AppendVector(
   A=ScalarParameter() {defaultValue=1, group=Turb, parameterName=NoiseTexture_Freq_U_01}
   B=ScalarParameter() {defaultValue=1, group=Turb, parameterName=NoiseTexture_Freq_V_01}
  )
  n5 = ScalarParameter() {group=Turb, parameterName=NoiseTexture_Speed_U_01}
  n6 = ScalarParameter() {group=Turb, parameterName=NoiseTexture_Speed_V_01}
  n7 = TextureSampleParameter2D(UVs=Panner(Coordinate=Multiply(A=@n3, B=@n4), Time=Time(), Speed=AppendVector(A=@n5, B=@n6)) {speedY=1}) {
   group=Turb
   parameterName=NoiseTexture01
   samplerType=SAMPLERTYPE_LinearColor
   texture=&T_NoisePack_LinearColor
  }
  n8 = Multiply(A=LinearInterpolate(Alpha=@n7.R) {constA=-1, constB=0.5}, B=ScalarParameter() {defaultValue=0.2, group=Turb, parameterName=Noise_Intensity}) {constB=0.2}
  Add(A=AppendVector(A=Add(A=@n1) {constB=0}, B=Add(A=@n2, B=@n8)), B=Constant2Vector() {g=0.1}) {constB=0}
  n9 = ParticleColor()
  n10 = TextureSampleParameter2D(UVs=Add(A=TextureCoordinate(), B=@n8)) {
   group=Opacity
   parameterName=BaseTexture
   samplerSource=SSM_Clamp_WorldGroupSettings
   texture=/Game/Effects/Fort_Effects/Textures/Mask/T_HalfSphere_D
  }
  n11 = Saturate(None=Multiply(A=Saturate(None=@n10.G), B=ScalarParameter() {defaultValue=1, group=Opacity, parameterName=Alpha_Multiply}))
  n12 = Saturate(None=Power(Base=Multiply(A=@n9.A, B=@n11), Exp=ScalarParameter() {defaultValue=2, group=Opacity, parameterName=Alpha_Power}))
  n13 = TextureCoordinate()
  n14 = Multiply(A=ComponentMask(None=@n13) {g=true}, B=OneMinus(None=ComponentMask(None=@n13) {g=true}))
  n15 = Saturate(None=Power(Base=Multiply(A=@n14) {constB=20}))
  n16 = ScalarParameter() {group=Opacity, parameterName=MeshFade}
  DepthFade(
   Opacity=Multiply(
    A=Multiply(A=@n12, B=@n15)
    B=MaterialFunctionCall("FadeSensitivity (optional)"=@n16) {materialFunction=&FadeBasedOnViewAngle}.Result
   )
   FadeDistance=ScalarParameter() {defaultValue=50, group=Opacity, parameterName=DepthFade}
  ) {fadeDistanceDefault=50}
  n17 = VectorParameter() {group=EmissiveColor, parameterName=Color, defaultValue=<rgba 0.0436 0 0.2917 1>}
  n18 = TextureCoordinate()
  n19 = AppendVector(
   A=ScalarParameter() {defaultValue=1, group=EmissiveColor, parameterName=ColorNoise_U_Freq_02}
   B=ScalarParameter() {defaultValue=1, group=EmissiveColor, parameterName=ColorNoise_V_Freq_02}
  )
  n20 = ScalarParameter() {group=EmissiveColor, parameterName=ColorNoise_U_Speed_02}
  n21 = ScalarParameter() {group=EmissiveColor, parameterName=ColorNoise_V_Speed_02}
  n22 = TextureSampleParameter2D(UVs=Panner(Coordinate=Multiply(A=@n18, B=@n19), Speed=AppendVector(A=@n20, B=@n21)) {speedY=1}) {
   group=EmissiveColor
   parameterName=ColorNoise_02
   samplerType=SAMPLERTYPE_Grayscale
   texture=&LowResBlurredNoise
  }
  n23 = TextureCoordinate()
  n24 = AppendVector(
   A=ScalarParameter() {defaultValue=1, group=EmissiveColor, parameterName=ColorNoise_U_Freq_01}
   B=ScalarParameter() {defaultValue=1, group=EmissiveColor, parameterName=ColorNoise_V_Freq_01}
  )
  n25 = ScalarParameter() {group=EmissiveColor, parameterName=ColorNoise_U_Speed_01}
  n26 = ScalarParameter() {group=EmissiveColor, parameterName=ColorNoise_V_Speed_01}
  n27 = TextureSampleParameter2D(UVs=Panner(Coordinate=Multiply(A=@n23, B=@n24), Speed=AppendVector(A=@n25, B=@n26)) {speedY=1}) {
   group=EmissiveColor
   parameterName=ColorNoise_01
   texture=/Game/Gadgets/Assets/VinderTech_GliderChute/Glider_Cube_Queen_Female/FX/T_Cheap_blue_space_color
  }
  n28 = ScalarParameter() {group=EmissiveColor, parameterName=Color_Desaturation}
  n29 = Add(A=Multiply(A=@n17.RGB, B=@n22.R), B=Add(A=@n17.RGB, B=Desaturation(None=@n27.RGB, Fraction=@n28)) {constB=0})
  n30 = ScalarParameter() {group=EmissiveColor, parameterName=Color_Global_Desaturation}
  MaterialFunctionCall(Emissive=Multiply(A=Desaturation(None=@n29, Fraction=@n30), B=@n9.RGB)) {materialFunction=&MF_FX_EmissiveAdjustments}
 ]
}
