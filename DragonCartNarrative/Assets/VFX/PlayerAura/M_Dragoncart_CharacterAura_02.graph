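# Translucent unlit particle material for the Dragoncart character aura, drawing tinted, smoky sprites or ribbons.
# A noise-distorted smoke flipbook and scrolling chromatic smoke build color and opacity, with per-particle random offsets and sine flicker.
# On deferred shading, a scene-normal Fresnel, limited by depth proximity, adds to mask and glow.
# Epic quality adds optional dithered refraction, while switchable near-camera fading also collapses vertices toward the particle center.
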Material /DragonCartNarrative/Assets/VFX/PlayerAura/M_Dragoncart_CharacterAura_02
graph {
 properties {
  bUsedWithNiagaraRibbons=true
  bUsedWithNiagaraSprites=true
  bUsedWithParticleSprites=true
  blendMode=BLEND_Translucent
  pixelDepthOffsetMode=PDOM_Legacy
  refractionMethod=RM_IndexOfRefraction
  shadingModel=MSM_Unlit
  usedShadingModels=Unlit
 }
 nodes [
  n1 = SceneTexture() {sceneTextureId=PPI_WorldNormal}
  n2 = ScalarParameter() {defaultValue=500, parameterName=FadeInDistance}
  n3 = Multiply(
   A=Fresnel(
    ExponentIn=ScalarParameter() {defaultValue=1.5, parameterName=WorldNormalFresnelExp}
    Normal=ComponentMask(None=@n1.Color) {b=true, g=true, r=true}
   ) {baseReflectFraction=0, exponent=1.5}
   B=OneMinus(None=DepthFade(FadeDistance=@n2) {fadeDistanceDefault=500})
  )
  n4 = Constant()
  n5 = ShadingPathSwitch(Default=@n3, Deferred=@n3, Forward=@n4, Mobile=@n4)
  n6 = Add(B=Multiply(A=ScalarParameter() {defaultValue=1, parameterName=WorldNormalFresnelMult}, B=@n5)) {constA=1}
  n7 = ParticleColor()
  n8 = TextureCoordinate() {uTiling=1.2, vTiling=1.2}
  n9 = DynamicParameter() {defaultValue=<rgba 0 1 1 1>, paramNames [Random,Random2,Flicker,Param4]}
  n10 = Multiply(A=Time(), B=ScalarParameter() {defaultValue=1, parameterName=PanSpeed})
  n11 = TextureSample(UVs=Panner(Coordinate=Add(A=@n8, B=@n9.Random2), Time=@n10) {speedX=-0.02, speedY=-0.05}) {texture=&T_Inky_Smoke_Tile_Chromatic}
  n12 = MaterialFunctionCall!(Use Alternate UVs, Alternate UVs) {materialFunction=/Game/Athena/Effects/Materials/Geode/Functions/MF_Geode_Crack_Colors}
  n13 = Constant() {r=2}
  n14 = ComponentMask(None=@n11.RGB) {g=true, r=true}
  n15 = Multiply(A=@n14, B=Constant2Vector() {g=0.05, r=0.02})
  n16 = Add(
   A=MaterialFunctionCall("Animation  Phase (0-1)"=@n9.Random, Number of Rows=@n13, Number of Columns=@n13) {materialFunction=&FlipBook}.UVs
   B=@n15
  )
  n17 = TextureCoordinate()
  n18 = Constant() {r=0.5}
  n19 = Multiply(
   A=Add(A=@n5, B=TextureSample(UVs=@n16) {texture=/Game/Effects/Fort_Effects/Textures/T_InkySmoke_Flipbook}.G)
   B=SphereMask(A=Add(A=@n15, B=@n17), B=@n18, Radius=@n18) {hardnessPercent=0}
  )
  n20 = Add(A=@n9.Random, B=TextureCoordinate() {uTiling=1.5, vTiling=1.5})
  n21 = Multiply(A=@n14, B=Constant2Vector() {g=-0.1, r=-0.02})
  n22 = TextureSample(UVs=Panner(Coordinate=Add(A=@n20, B=@n21), Time=@n10) {speedX=0.05, speedY=0.12}) {texture=&T_Inky_Smoke_Tile_Chromatic}
  n23 = LinearInterpolate(A=@n12.Cloud Color, B=@n12.Cloud Color, Alpha=LinearInterpolate(A=Multiply(A=@n19, B=@n19), B=@n22.G) {constAlpha=0.7})
  MaterialFunctionCall(Emissive=Multiply(A=Multiply(A=@n6, B=@n7.RGB), B=Add(A=@n11.RGB, B=@n23))) {materialFunction=&MF_FX_EmissiveAdjustments}
  n24 = Constant() {r=1}
  n25 = ScalarParameter() {defaultValue=0.5, parameterName=RefractionAdd}
  n26 = LinearInterpolate(Alpha=@n9.Random2) {constA=-1}
  n27 = Multiply(A=@n22.R) {constB=2}
  n28 = Multiply(A=Time(), B=LinearInterpolate(Alpha=@n9.Random2) {constA=1.2, constB=1.7})
  n29 = Constant() {r=4}
  n30 = Divide(
   A=Add(A=MaterialFunctionCall!(Time=@n28) {materialFunction=/Game/Effects/Fort_Effects/Materials/Geode/_Functions/MF_SineVibration}.Result, B=@n29) {constB=5}
   B=@n29
  )
  n31 = ScalarParameter() {defaultValue=50, parameterName=DepthFadeOutDist}
  n32 = Multiply(
   A=Multiply(A=@n7.A, B=LinearInterpolate(B=@n30, Alpha=@n9.Flicker) {constA=1})
   B=ShadingPathSwitch(Default=DepthFade(Opacity=@n19, FadeDistance=@n31) {fadeDistanceDefault=50}, Forward=@n19, Mobile=@n19)
  )
  n33 = LinearInterpolate(B=Add(A=Multiply(A=@n25, B=@n26)), Alpha=Add(A=@n32, B=Multiply(A=@n32, B=@n27))) {constA=1}
  QualitySwitch(
   Default=@n24
   Epic=StaticSwitchParameter(
    True=@n24
    False=MaterialFunctionCall!(RefractiveIndex=@n33) {materialFunction=/Game/Effects/Fort_Effects/Materials/_Functions/MF_Dithered_Refraction}.Result
   ) {parameterName=Remove_Refraction}
  )
  n34 = ScalarParameter() {defaultValue=1024, parameterName=FadeNearCameraDistance}
  n35 = SphereMask(A=PixelDepth().R, B=Constant(), Radius=Subtract(A=@n34) {constB=250}) {hardnessPercent=10}
  n36 = ScalarParameter() {parameterName=FadeSensitivity}
  n37 = LinearInterpolate(
   A=@n32
   B=Multiply(
    A=@n32
    B=MaterialFunctionCall("FadeSensitivity (optional)"=@n36) {materialFunction=&FadeBasedOnViewAngle}.Result
   )
   Alpha=ScalarParameter() {parameterName=Fade based on View Angle}
  )
  StaticSwitchParameter(True=Multiply(A=LinearInterpolate(Alpha=@n35) {constA=1, constB=0}, B=@n37), False=@n37) {parameterName=UseNearFade}
  n38 = Subtract(A=WorldPosition().XYZ, B=ParticlePositionWS())
  n39 = SphereMask(A=PixelDepth().R, B=Constant(), Radius=@n34) {hardnessPercent=10}
  StaticSwitchParameter(True=LinearInterpolate(A=Multiply(A=@n38) {constB=-1}, Alpha=OneMinus(None=@n39)) {constB=0}, False=Constant()) {parameterName=UseNearFade}
  TextureSample() {samplerType=SAMPLERTYPE_Normal, texture=&T_Sphere_N}
 ]
}
